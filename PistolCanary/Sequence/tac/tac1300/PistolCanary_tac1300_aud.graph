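# Audio level sequence for shot tac1300 of the PistolCanary sequence, covering a tentacle boss moment.
# Two master TentacleAudio tracks each play one section, the second-listed starting first and the longer one overlapping it shortly after, both at default volume and pitch.
# A possessed BGA_Tentacle actor, bound as TentacleBoss, carries an event trigger track; the file does not show what that event calls.
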LevelSequence /PistolCanary/Sequence/tac/tac1300/PistolCanary_tac1300_aud
bindings [TentacleBoss {
  id=#1
  possessed_class=BGA_Tentacle_C
  tracks [{
    type=MovieSceneEventTrack
    sections [{
      type=MovieSceneEventTriggerSection
      channels [None {type=MovieSceneScriptingEventChannel, const ["<Struct 'MovieSceneEvent' (0x852ff398660) {}>", 1]}]
     }]
   }]
 }]
display_rate {denominator=1, numerator=30}
tick_resolution {denominator=1, numerator=24000}
tracks [
 {
  type=MovieSceneAudioTrack
  display_name=TentacleAudio
  sections [{
    type=MovieSceneAudioSection
    end_frame=391
    start_frame=27
    channels [
     Volume {type=MovieSceneScriptingFloatChannel, default=1}
     Pitch {type=MovieSceneScriptingFloatChannel, default=1}
     AttachActor {type=MovieSceneScriptingActorReferenceChannel}
    ]
   }]
 }
 {
  type=MovieSceneAudioTrack
  display_name=TentacleAudio
  sections [{
    type=MovieSceneAudioSection
    end_frame=193
    start_frame=0
    channels [
     Volume {type=MovieSceneScriptingFloatChannel, default=1}
     Pitch {type=MovieSceneScriptingFloatChannel, default=1}
     AttachActor {type=MovieSceneScriptingActorReferenceChannel}
    ]
   }]
 }
]
